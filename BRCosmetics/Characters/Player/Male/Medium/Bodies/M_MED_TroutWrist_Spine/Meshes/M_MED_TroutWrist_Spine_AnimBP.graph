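# Post-process-style AnimBP for the TroutWrist Spine cosmetic body, adjusting its claw and belt bones on top of the parent mesh's pose.
# Copied pose feeds metacarpal-pair constraints on claw bases (ClawsConstraint), then additive claw-open offsets (Claws open).
# Belt bones then follow weighted pelvis, spine and thigh constraints, approximating skinning.
# Other visible crouch, skirt, AIM-bone and emote chains start without a pose input, so appear to be disconnected variants.
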AnimBlueprint /BRCosmetics/Characters/Player/Male/Medium/Bodies/M_MED_TroutWrist_Spine/Meshes/M_MED_TroutWrist_Spine_AnimBP
target_skeleton=&/M_MED_TroutWrist_Spine_Skeleton
graphs [
 AnimGraph {nodes [
   n1 = Get bIsCrouching()
   "Constraint - dyn_belt_f_01 (PPPPPPPP)"!(
    ComponentPose
    bAlphaBoolEnabled:bool=@n1
    ConstraintWeights_0=0.2999
    ConstraintWeights_1:float=0.1647
    ConstraintWeights_2:float=0.2651
    ConstraintWeights_3:float=0.0976
    ConstraintWeights_4:float=0.0325
    ConstraintWeights_5:float=0.0058
    ConstraintWeights_6:float=0.0701
    ConstraintWeights_7:float=0.0644
   ) -> (Pose) {
    alphaInputType=Bool
    alphaBoolBlend {blendOption=Linear}
    boneToModify {boneName=dyn_belt_f_01}
    constraintSetup [
     {targetBone {boneName=pelvis}}
     {targetBone {boneName=spine_01}}
     {targetBone {boneName=spine_02}}
     {targetBone {boneName=spine_03}}
     {targetBone {boneName=spine_04}}
     {targetBone {boneName=thigh_l}}
     {targetBone {boneName=thigh_twist_01_l}}
     {targetBone {boneName=thigh_twist_01_r}}
    ]
    constraintWeights [1,1,1,1,1,1,1,1]
   }
   n2 = "Constraint - dyn_belt_f_01 (PPPPPPPP)"(
    bAlphaBoolEnabled=@n1
    ConstraintWeights_0=0.2014
    ConstraintWeights_1=0.1604
    ConstraintWeights_2=0.3404
    ConstraintWeights_3=0.1422
    ConstraintWeights_4=0.0547
    ConstraintWeights_5=0.0033
    ConstraintWeights_6=0.0517
    ConstraintWeights_7=0.0459
   ) {
    alphaInputType=Bool
    alphaBoolBlend {blendOption=Linear}
    boneToModify {boneName=dyn_belt_f_01}
    constraintSetup [
     {targetBone {boneName=pelvis}}
     {targetBone {boneName=spine_01}}
     {targetBone {boneName=spine_02}}
     {targetBone {boneName=spine_03}}
     {targetBone {boneName=spine_04}}
     {targetBone {boneName=thigh_l}}
     {targetBone {boneName=thigh_twist_01_l}}
     {targetBone {boneName=thigh_twist_01_r}}
    ]
    constraintWeights [1,1,1,1,1,1,1,1]
   }
   n3 = "Constraint - dyn_belt_r_01 (PPPPPPPP)"!(
    ComponentPose=@n2
    bAlphaBoolEnabled:bool=@n1
    ConstraintWeights_0=0.3102
    ConstraintWeights_1:float=0.2528
    ConstraintWeights_2:float=0.2053
    ConstraintWeights_3:float=0.0491
    ConstraintWeights_4:float=0.008
    ConstraintWeights_5:float=0.0138
    ConstraintWeights_6:float=0.0346
    ConstraintWeights_7:float=0.1262
   ) -> (Pose) {
    alphaInputType=Bool
    alphaBoolBlend {blendOption=Linear}
    boneToModify {boneName=dyn_belt_r_01}
    constraintSetup [
     {targetBone {boneName=pelvis}}
     {targetBone {boneName=spine_01}}
     {targetBone {boneName=spine_02}}
     {targetBone {boneName=spine_03}}
     {targetBone {boneName=spine_04}}
     {targetBone {boneName=thigh_twist_01_l}}
     {targetBone {boneName=thigh_r}}
     {targetBone {boneName=thigh_twist_01_r}}
    ]
    constraintWeights [1,1,1,1,1,1,1,1]
   }
   n4 = "Constraint - dyn_belt_l_01 (PPPPPPPP)"!(
    ComponentPose=@n3
    bAlphaBoolEnabled:bool=@n1
    ConstraintWeights_0=0.2315
    ConstraintWeights_1:float=0.271
    ConstraintWeights_2:float=0.2802
    ConstraintWeights_3:float=0.0768
    ConstraintWeights_4:float=0.0108
    ConstraintWeights_5:float=0.0235
    ConstraintWeights_6:float=0.0982
    ConstraintWeights_7:float=0.0081
   ) -> (Pose) {
    alphaInputType=Bool
    alphaBoolBlend {blendOption=Linear}
    boneToModify {boneName=dyn_belt_l_01}
    constraintSetup [
     {targetBone {boneName=pelvis}}
     {targetBone {boneName=spine_01}}
     {targetBone {boneName=spine_02}}
     {targetBone {boneName=spine_03}}
     {targetBone {boneName=spine_04}}
     {targetBone {boneName=thigh_l}}
     {targetBone {boneName=thigh_twist_01_l}}
     {targetBone {boneName=thigh_twist_01_r}}
    ]
    constraintWeights [1,1,1,1,1,1,1,1]
   }
   "Constraint - dyn_belt_b_01 (PPPPPP)"!(
    ComponentPose=@n4
    bAlphaBoolEnabled:bool=@n1
    ConstraintWeights_0=0.5157
    ConstraintWeights_1:float=0.2398
    ConstraintWeights_2:float=0.053
    ConstraintWeights_3:float=0.1529
    ConstraintWeights_4:float=0.0192
    ConstraintWeights_5:float=0.0192
   ) -> (Pose) {
    alphaInputType=Bool
    alphaBoolBlend {blendOption=Linear}
    boneToModify {boneName=dyn_belt_b_01}
    constraintSetup [
     {targetBone {boneName=pelvis}}
     {targetBone {boneName=spine_01}}
     {targetBone {boneName=spine_02}}
     {targetBone {boneName=spine_03}}
     {targetBone {boneName=thigh_twist_01_l}}
     {targetBone {boneName=thigh_twist_01_r}}
    ]
    constraintWeights [1,1,1,1,1,1]
   }
   "Constraint - dyn_belt_f_01 (PPPPPPPP)"(
    bAlphaBoolEnabled=@n1
    ConstraintWeights_0=0.3717
    ConstraintWeights_1=0.1551
    ConstraintWeights_2=0.2045
    ConstraintWeights_3=0.0685
    ConstraintWeights_4=0.0206
    ConstraintWeights_5=0.008
    ConstraintWeights_6=0.0887
    ConstraintWeights_7=0.0829
   ) {
    alphaInputType=Bool
    alphaBoolBlend {blendOption=Linear}
    boneToModify {boneName=dyn_belt_f_01}
    constraintSetup [
     {targetBone {boneName=pelvis}}
     {targetBone {boneName=spine_01}}
     {targetBone {boneName=spine_02}}
     {targetBone {boneName=spine_03}}
     {targetBone {boneName=spine_04}}
     {targetBone {boneName=thigh_l}}
     {targetBone {boneName=thigh_twist_01_l}}
     {targetBone {boneName=thigh_twist_01_r}}
    ]
    constraintWeights [1,1,1,1,1,1,1,1]
   }
   n5 = Get bIsCrouching()
   "Constraint - dyn_belt_l_01_AIM (PPPPPPPP)"!(
    ComponentPose
    bAlphaBoolEnabled:bool=@n5
    ConstraintWeights_0=0.2245
    ConstraintWeights_1:float=0.2896
    ConstraintWeights_2:float=0.276
    ConstraintWeights_3:float=0.0729
    ConstraintWeights_4:float=0.0088
    ConstraintWeights_5:float=0.0249
    ConstraintWeights_6:float=0.0976
    ConstraintWeights_7:float=0.0057
   ) -> (Pose) {
    alphaInputType=Bool
    alphaBoolBlend {blendOption=Linear}
    boneToModify {boneName=dyn_belt_l_01_AIM}
    constraintSetup [
     {targetBone {boneName=pelvis}}
     {targetBone {boneName=spine_01}}
     {targetBone {boneName=spine_02}}
     {targetBone {boneName=spine_03}}
     {targetBone {boneName=spine_04}}
     {targetBone {boneName=thigh_l}}
     {targetBone {boneName=thigh_twist_01_l}}
     {targetBone {boneName=thigh_twist_01_r}}
    ]
    constraintWeights [1,1,1,1,1,1,1,1]
   }
   n6 = NOT Boolean(A=Get bIsCrouching())
   n7 = "Constraint - dyn_SkirtF00_f (PPPPPPPP)"!(
    ComponentPose
    bAlphaBoolEnabled:bool=@n6
    ConstraintWeights_0=0.4811
    ConstraintWeights_1:float=0.0853
    ConstraintWeights_2:float=0.0743
    ConstraintWeights_3:float=0.0202
    ConstraintWeights_4:float=0.0144
    ConstraintWeights_5:float=0.1578
    ConstraintWeights_6:float=0.0142
    ConstraintWeights_7:float=0.1528
   ) -> (Pose) {
    alphaInputType=Bool
    alphaBoolBlend {blendOption=Linear}
    boneToModify {boneName=dyn_SkirtF00_f}
    constraintSetup [
     {targetBone {boneName=pelvis}}
     {targetBone {boneName=spine_01}}
     {targetBone {boneName=spine_02}}
     {targetBone {boneName=spine_03}}
     {targetBone {boneName=thigh_l}}
     {targetBone {boneName=thigh_twist_01_l}}
     {targetBone {boneName=thigh_r}}
     {targetBone {boneName=thigh_twist_01_r}}
    ]
    constraintWeights [1,1,1,1,1,1,1,1]
   }
   n8 = "Constraint - dyn_SkirtA00_l (PPPPPPPP)"!(
    ComponentPose=@n7
    bAlphaBoolEnabled:bool=@n6
    ConstraintWeights_0=0.52
    ConstraintWeights_1:float=0.0775
    ConstraintWeights_2:float=0.0378
    ConstraintWeights_3:float=0.0072
    ConstraintWeights_4:float=0.0014
    ConstraintWeights_5:float=0.0481
    ConstraintWeights_6:float=0.2854
    ConstraintWeights_7:float=0.0225
   ) -> (Pose) {
    alphaInputType=Bool
    alphaBoolBlend {blendOption=Linear}
    boneToModify {boneName=dyn_SkirtA00_l}
    constraintSetup [
     {targetBone {boneName=pelvis}}
     {targetBone {boneName=spine_01}}
     {targetBone {boneName=spine_02}}
     {targetBone {boneName=spine_03}}
     {targetBone {boneName=spine_04}}
     {targetBone {boneName=thigh_l}}
     {targetBone {boneName=thigh_twist_01_l}}
     {targetBone {boneName=thigh_twist_01_r}}
    ]
    constraintWeights [1,1,1,1,1,1,1,1]
   }
   n9 = "Constraint - dyn_SkirtB00_l (PPPPPPP)"!(
    ComponentPose=@n8
    bAlphaBoolEnabled:bool=@n6
    ConstraintWeights_0=0.5709
    ConstraintWeights_1:float=0.03
    ConstraintWeights_2:float=0.0381
    ConstraintWeights_3:float=0.0093
    ConstraintWeights_4:float=0.0002
    ConstraintWeights_5:float=0.0395
    ConstraintWeights_6:float=0.3119
   ) -> (Pose) {
    alphaInputType=Bool
    alphaBoolBlend {blendOption=Linear}
    boneToModify {boneName=dyn_SkirtB00_l}
    constraintSetup [
     {targetBone {boneName=pelvis}}
     {targetBone {boneName=spine_01}}
     {targetBone {boneName=spine_02}}
     {targetBone {boneName=spine_03}}
     {targetBone {boneName=spine_05}}
     {targetBone {boneName=thigh_l}}
     {targetBone {boneName=thigh_twist_01_l}}
    ]
    constraintWeights [1,1,1,1,1,1,1]
   }
   n10 = "Constraint - dyn_SkirtC00_l (PPPPPPP)"!(
    ComponentPose=@n9
    bAlphaBoolEnabled:bool=@n6
    ConstraintWeights_0=0.6526
    ConstraintWeights_1:float=0.1016
    ConstraintWeights_2:float=0.0523
    ConstraintWeights_3:float=0.1169
    ConstraintWeights_4:float=0.0058
    ConstraintWeights_5:float=0.0004
    ConstraintWeights_6:float=0.0705
   ) -> (Pose) {
    alphaInputType=Bool
    alphaBoolBlend {blendOption=Linear}
    boneToModify {boneName=dyn_SkirtC00_l}
    constraintSetup [
     {targetBone {boneName=pelvis}}
     {targetBone {boneName=spine_01}}
     {targetBone {boneName=spine_02}}
     {targetBone {boneName=spine_03}}
     {targetBone {boneName=spine_05}}
     {targetBone {boneName=thigh_l}}
     {targetBone {boneName=thigh_twist_01_l}}
    ]
    constraintWeights [1,1,1,1,1,1,1]
   }
   n11 = "Constraint - dyn_SkirtC00_r (PPPPPPPP)"!(
    ComponentPose=@n10
    bAlphaBoolEnabled:bool=@n6
    ConstraintWeights_0=0.6402
    ConstraintWeights_1:float=0.1086
    ConstraintWeights_2:float=0.0514
    ConstraintWeights_3:float=0.1161
    ConstraintWeights_4:float=0.0045
    ConstraintWeights_5:float=0.0002
    ConstraintWeights_6:float=0.0011
    ConstraintWeights_7:float=0.0779
   ) -> (Pose) {
    alphaInputType=Bool
    alphaBoolBlend {blendOption=Linear}
    boneToModify {boneName=dyn_SkirtC00_r}
    constraintSetup [
     {targetBone {boneName=pelvis}}
     {targetBone {boneName=spine_01}}
     {targetBone {boneName=spine_02}}
     {targetBone {boneName=spine_03}}
     {targetBone {boneName=spine_05}}
     {targetBone {boneName=thigh_twist_01_l}}
     {targetBone {boneName=thigh_r}}
     {targetBone {boneName=thigh_twist_01_r}}
    ]
    constraintWeights [1,1,1,1,1,1,1,1]
   }
   n12 = "Constraint - dyn_SkirtB00_r (PPPPPPP)"!(
    ComponentPose=@n11
    bAlphaBoolEnabled:bool=@n6
    ConstraintWeights_0=0.5655
    ConstraintWeights_1:float=0.0304
    ConstraintWeights_2:float=0.0386
    ConstraintWeights_3:float=0.0094
    ConstraintWeights_4:float=0.0002
    ConstraintWeights_5:float=0.04
    ConstraintWeights_6:float=0.3159
   ) -> (Pose) {
    alphaInputType=Bool
    alphaBoolBlend {blendOption=Linear}
    boneToModify {boneName=dyn_SkirtB00_r}
    constraintSetup [
     {targetBone {boneName=pelvis}}
     {targetBone {boneName=spine_01}}
     {targetBone {boneName=spine_02}}
     {targetBone {boneName=spine_03}}
     {targetBone {boneName=spine_05}}
     {targetBone {boneName=thigh_r}}
     {targetBone {boneName=thigh_twist_01_r}}
    ]
    constraintWeights [1,1,1,1,1,1,1]
   }
   n13 = "Constraint - dyn_SkirtA00_r (PPPPPPPP)"!(
    ComponentPose=@n12
    bAlphaBoolEnabled:bool=@n6
    ConstraintWeights_0=0.4987
    ConstraintWeights_1:float=0.0936
    ConstraintWeights_2:float=0.0498
    ConstraintWeights_3:float=0.0087
    ConstraintWeights_4:float=0.0017
    ConstraintWeights_5:float=0.0263
    ConstraintWeights_6:float=0.0588
    ConstraintWeights_7:float=0.2623
   ) -> (Pose) {
    alphaInputType=Bool
    alphaBoolBlend {blendOption=Linear}
    boneToModify {boneName=dyn_SkirtA00_r}
    constraintSetup [
     {targetBone {boneName=pelvis}}
     {targetBone {boneName=spine_01}}
     {targetBone {boneName=spine_02}}
     {targetBone {boneName=spine_03}}
     {targetBone {boneName=spine_04}}
     {targetBone {boneName=thigh_twist_01_l}}
     {targetBone {boneName=thigh_r}}
     {targetBone {boneName=thigh_twist_01_r}}
    ]
    constraintWeights [1,1,1,1,1,1,1,1]
   }
   n14 = Get bIsCrouching()
   n15 = "Constraint - dyn_SkirtF00_f (P)"!(ComponentPose=@n13, bAlphaBoolEnabled:bool=@n14, ConstraintWeights_0=1) -> (Pose) {
    alphaInputType=Bool
    alphaBoolBlend {blendOption=Linear}
    boneToModify {boneName=dyn_SkirtF00_f}
    constraintSetup [{offsetOption=None, targetBone {boneName=dyn_SkirtF00_f_AIM}}]
    constraintWeights [1]
   }
   n16 = "Constraint - dyn_SkirtA00_l (P)"!(ComponentPose=@n15, bAlphaBoolEnabled:bool=@n14, ConstraintWeights_0=1) -> (Pose) {
    alphaInputType=Bool
    alphaBoolBlend {blendOption=Linear}
    boneToModify {boneName=dyn_SkirtA00_l}
    constraintSetup [{offsetOption=None, targetBone {boneName=dyn_SkirtA00_l_AIM}}]
    constraintWeights [1]
   }
   n17 = "Constraint - dyn_SkirtB00_l (P)"!(ComponentPose=@n16, bAlphaBoolEnabled:bool=@n14, ConstraintWeights_0=1) -> (Pose) {
    alphaInputType=Bool
    alphaBoolBlend {blendOption=Linear}
    boneToModify {boneName=dyn_SkirtB00_l}
    constraintSetup [{offsetOption=None, targetBone {boneName=dyn_SkirtB00_l_AIM}}]
    constraintWeights [1]
   }
   n18 = "Constraint - dyn_SkirtC00_l (P)"!(ComponentPose=@n17, bAlphaBoolEnabled:bool=@n14, ConstraintWeights_0=1) -> (Pose) {
    alphaInputType=Bool
    alphaBoolBlend {blendOption=Linear}
    boneToModify {boneName=dyn_SkirtC00_l}
    constraintSetup [{offsetOption=None, targetBone {boneName=dyn_SkirtC00_l_AIM}}]
    constraintWeights [1]
   }
   n19 = "Constraint - dyn_SkirtC00_r (P)"!(ComponentPose=@n18, bAlphaBoolEnabled:bool=@n14, ConstraintWeights_0=1) -> (Pose) {
    alphaInputType=Bool
    alphaBoolBlend {blendOption=Linear}
    boneToModify {boneName=dyn_SkirtC00_r}
    constraintSetup [{offsetOption=None, targetBone {boneName=dyn_SkirtC00_r_AIM}}]
    constraintWeights [1]
   }
   n20 = "Constraint - dyn_SkirtB00_r (P)"!(ComponentPose=@n19, bAlphaBoolEnabled:bool=@n14, ConstraintWeights_0=1) -> (Pose) {
    alphaInputType=Bool
    alphaBoolBlend {blendOption=Linear}
    boneToModify {boneName=dyn_SkirtB00_r}
    constraintSetup [{offsetOption=None, targetBone {boneName=dyn_SkirtB00_r_AIM}}]
    constraintWeights [1]
   }
   n21 = "Constraint - dyn_SkirtA00_r (P)"!(ComponentPose=@n20, bAlphaBoolEnabled:bool=@n14, ConstraintWeights_0=1) -> (Pose) {
    alphaInputType=Bool
    alphaBoolBlend {blendOption=Linear}
    boneToModify {boneName=dyn_SkirtA00_r}
    constraintSetup [{offsetOption=None, targetBone {boneName=dyn_SkirtA00_r_AIM}}]
    constraintWeights [1]
   }
   n22 = "Constraint - dyn_SkirtF00_f_AIM (PPPPPPPP)"!(
    ComponentPose=@n21
    Alpha=1
    ConstraintWeights_0=0.3717
    ConstraintWeights_1:float=0.1551
    ConstraintWeights_2:float=0.2045
    ConstraintWeights_3:float=0.0685
    ConstraintWeights_4:float=0.0206
    ConstraintWeights_5:float=0.008
    ConstraintWeights_6:float=0.0887
    ConstraintWeights_7:float=0.0829
   ) -> (Pose) {
    alphaBoolBlend {blendOption=Linear}
    boneToModify {boneName=dyn_SkirtF00_f_AIM}
    constraintSetup [
     {targetBone {boneName=pelvis}}
     {targetBone {boneName=spine_01}}
     {targetBone {boneName=spine_02}}
     {targetBone {boneName=spine_03}}
     {targetBone {boneName=spine_04}}
     {targetBone {boneName=thigh_l}}
     {targetBone {boneName=thigh_twist_01_l}}
     {targetBone {boneName=thigh_twist_01_r}}
    ]
    constraintWeights [1,1,1,1,1,1,1,1]
   }
   n23 = "Constraint - dyn_SkirtA00_l_AIM (PPPPPPPP)"!(
    ComponentPose=@n22
    Alpha=1
    ConstraintWeights_0=0.3382
    ConstraintWeights_1:float=0.2586
    ConstraintWeights_2:float=0.1798
    ConstraintWeights_3:float=0.0423
    ConstraintWeights_4:float=0.006
    ConstraintWeights_5:float=0.0322
    ConstraintWeights_6:float=0.1322
    ConstraintWeights_7:float=0.0107
   ) -> (Pose) {
    alphaBoolBlend {blendOption=Linear}
    boneToModify {boneName=dyn_SkirtA00_l_AIM}
    constraintSetup [
     {targetBone {boneName=pelvis}}
     {targetBone {boneName=spine_01}}
     {targetBone {boneName=spine_02}}
     {targetBone {boneName=spine_03}}
     {targetBone {boneName=spine_04}}
     {targetBone {boneName=thigh_l}}
     {targetBone {boneName=thigh_twist_01_l}}
     {targetBone {boneName=thigh_twist_01_r}}
    ]
    constraintWeights [1,1,1,1,1,1,1,1]
   }
   n24 = "Constraint - dyn_SkirtB00_l_AIM (PPPPPPP)"!(
    ComponentPose=@n23
    Alpha=1
    ConstraintWeights_0=0.1856
    ConstraintWeights_1:float=0.4677
    ConstraintWeights_2:float=0.1355
    ConstraintWeights_3:float=0.0933
    ConstraintWeights_4:float=0.0012
    ConstraintWeights_5:float=0.0162
    ConstraintWeights_6:float=0.1004
   ) -> (Pose) {
    alphaBoolBlend {blendOption=Linear}
    boneToModify {boneName=dyn_SkirtB00_l_AIM}
    constraintSetup [
     {targetBone {boneName=pelvis}}
     {targetBone {boneName=spine_01}}
     {targetBone {boneName=spine_02}}
     {targetBone {boneName=spine_03}}
     {targetBone {boneName=spine_04}}
     {targetBone {boneName=thigh_l}}
     {targetBone {boneName=thigh_twist_01_l}}
    ]
    constraintWeights [1,1,1,1,1,1,1]
   }
   n25 = "Constraint - dyn_SkirtC00_l_AIM (PPPPP)"!(
    ComponentPose=@n24
    Alpha=1
    ConstraintWeights_0=0.2874
    ConstraintWeights_1:float=0.4674
    ConstraintWeights_2:float=0.0634
    ConstraintWeights_3:float=0.1484
    ConstraintWeights_4:float=0.0335
   ) -> (Pose) {
    alphaBoolBlend {blendOption=Linear}
    boneToModify {boneName=dyn_SkirtC00_l_AIM}
    constraintSetup [
     {targetBone {boneName=pelvis}}
     {targetBone {boneName=spine_01}}
     {targetBone {boneName=spine_02}}
     {targetBone {boneName=spine_03}}
     {targetBone {boneName=thigh_twist_01_l}}
    ]
    constraintWeights [1,1,1,1,1]
   }
   n26 = "Constraint - dyn_SkirtC00_r_AIM (PPPPPPPP)"!(
    ComponentPose=@n25
    Alpha=1
    ConstraintWeights_0=0.3018
    ConstraintWeights_1:float=0.4279
    ConstraintWeights_2:float=0.0789
    ConstraintWeights_3:float=0.1556
    ConstraintWeights_4:float=0.0002
    ConstraintWeights_5:float=0.0001
    ConstraintWeights_6:float=0.0002
    ConstraintWeights_7:float=0.0353
   ) -> (Pose) {
    alphaBoolBlend {blendOption=Linear}
    boneToModify {boneName=dyn_SkirtC00_r_AIM}
    constraintSetup [
     {targetBone {boneName=pelvis}}
     {targetBone {boneName=spine_01}}
     {targetBone {boneName=spine_02}}
     {targetBone {boneName=spine_03}}
     {targetBone {boneName=spine_04}}
     {targetBone {boneName=spine_05}}
     {targetBone {boneName=thigh_twist_01_l}}
     {targetBone {boneName=thigh_twist_01_r}}
    ]
    constraintWeights [1,1,1,1,1,1,1,1]
   }
   n27 = "Constraint - dyn_SkirtB00_r_AIM (PPPPPPPP)"!(
    ComponentPose=@n26
    Alpha=1
    ConstraintWeights_0=0.1915
    ConstraintWeights_1:float=0.4736
    ConstraintWeights_2:float=0.1426
    ConstraintWeights_3:float=0.0853
    ConstraintWeights_4:float=0.0029
    ConstraintWeights_5:float=0.0002
    ConstraintWeights_6:float=0.0155
    ConstraintWeights_7:float=0.0883
   ) -> (Pose) {
    alphaBoolBlend {blendOption=Linear}
    boneToModify {boneName=dyn_SkirtB00_r_AIM}
    constraintSetup [
     {targetBone {boneName=pelvis}}
     {targetBone {boneName=spine_01}}
     {targetBone {boneName=spine_02}}
     {targetBone {boneName=spine_03}}
     {targetBone {boneName=spine_04}}
     {targetBone {boneName=spine_05}}
     {targetBone {boneName=thigh_r}}
     {targetBone {boneName=thigh_twist_01_r}}
    ]
    constraintWeights [1,1,1,1,1,1,1,1]
   }
   "Constraint - dyn_SkirtA00_r_AIM (PPPPPPPP)"!(
    ComponentPose=@n27
    Alpha=1
    ConstraintWeights_0=0.3102
    ConstraintWeights_1:float=0.2528
    ConstraintWeights_2:float=0.2053
    ConstraintWeights_3:float=0.0491
    ConstraintWeights_4:float=0.008
    ConstraintWeights_5:float=0.0138
    ConstraintWeights_6:float=0.0346
    ConstraintWeights_7:float=0.1262
   ) -> (Pose) {
    alphaBoolBlend {blendOption=Linear}
    boneToModify {boneName=dyn_SkirtA00_r_AIM}
    constraintSetup [
     {targetBone {boneName=pelvis}}
     {targetBone {boneName=spine_01}}
     {targetBone {boneName=spine_02}}
     {targetBone {boneName=spine_03}}
     {targetBone {boneName=spine_04}}
     {targetBone {boneName=thigh_twist_01_l}}
     {targetBone {boneName=thigh_r}}
     {targetBone {boneName=thigh_twist_01_r}}
    ]
    constraintWeights [1,1,1,1,1,1,1,1]
   }
   n28 = "Constraint - dyn_belt_f_01 (PPPPPP)"!(
    ComponentPose
    Alpha=1
    ConstraintWeights_0=0.8836
    ConstraintWeights_1:float=0.0396
    ConstraintWeights_2:float=0.0507
    ConstraintWeights_3:float=0.0178
    ConstraintWeights_4:float=0.0052
    ConstraintWeights_5:float=0.0031
   ) -> (Pose) {
    alphaBoolBlend {blendOption=Linear}
    boneToModify {boneName=dyn_belt_f_01}
    constraintSetup [
     {targetBone {boneName=pelvis}}
     {targetBone {boneName=spine_01}}
     {targetBone {boneName=spine_02}}
     {targetBone {boneName=spine_03}}
     {targetBone {boneName=thigh_l}}
     {targetBone {boneName=thigh_r}}
    ]
    constraintWeights [1,1,1,1,1,1]
   }
   n29 = "Constraint - dyn_belt_r_01 (PPPPPPPP)"!(
    ComponentPose=@n28
    Alpha:float=1
    ConstraintWeights_0=0.8725
    ConstraintWeights_1=0.0475
    ConstraintWeights_2=0.0309
    ConstraintWeights_3=0.0067
    ConstraintWeights_4=0.0011
    ConstraintWeights_5=0.0004
    ConstraintWeights_6=0.0215
    ConstraintWeights_7=0.0194
   ) -> (Pose) {
    alphaBoolBlend {blendOption=Linear}
    boneToModify {boneName=dyn_belt_r_01}
    constraintSetup [
     {targetBone {boneName=pelvis}}
     {targetBone {boneName=spine_01}}
     {targetBone {boneName=spine_02}}
     {targetBone {boneName=spine_03}}
     {targetBone {boneName=spine_04}}
     {targetBone {boneName=thigh_twist_01_l}}
     {targetBone {boneName=thigh_r}}
     {targetBone {boneName=thigh_twist_01_r}}
    ]
    constraintWeights [1,1,1,1,1,1,1,1]
   }
   "Constraint - dyn_belt_l_01 (PPPPPPPP)"!(
    ComponentPose=@n29
    Alpha:float=1
    ConstraintWeights_0=0.8149
    ConstraintWeights_1=0.0726
    ConstraintWeights_2=0.0445
    ConstraintWeights_3=0.0103
    ConstraintWeights_4=0.0016
    ConstraintWeights_5=0.0269
    ConstraintWeights_6=0.0285
    ConstraintWeights_7=0.0007
   ) -> (Pose) {
    alphaBoolBlend {blendOption=Linear}
    boneToModify {boneName=dyn_belt_l_01}
    constraintSetup [
     {targetBone {boneName=pelvis}}
     {targetBone {boneName=spine_01}}
     {targetBone {boneName=spine_02}}
     {targetBone {boneName=spine_03}}
     {targetBone {boneName=spine_04}}
     {targetBone {boneName=thigh_l}}
     {targetBone {boneName=thigh_twist_01_l}}
     {targetBone {boneName=thigh_twist_01_r}}
    ]
    constraintWeights [1,1,1,1,1,1,1,1]
   }
   "Constraint - clawbase_01_r (PPPPPPPP)"!(
    ComponentPose
    Alpha=1
    ConstraintWeights_0=0.0919
    ConstraintWeights_1:float=0.4038
    ConstraintWeights_2:float=0.0077
    ConstraintWeights_3:float=0.1334
    ConstraintWeights_4:float=0.3204
    ConstraintWeights_5:float=0.0025
    ConstraintWeights_6:float=0.0316
    ConstraintWeights_7:float=0.0087
   ) -> (Pose) {
    alphaBoolBlend {blendOption=Linear}
    boneToModify {boneName=clawbase_01_r}
    constraintSetup [
     {targetBone {boneName=index_metacarpal_r}}
     {targetBone {boneName=index_01_r}}
     {targetBone {boneName=index_02_r}}
     {targetBone {boneName=middle_metacarpal_r}}
     {targetBone {boneName=middle_01_r}}
     {targetBone {boneName=middle_02_r}}
     {targetBone {boneName=ring_metacarpal_r}}
     {targetBone {boneName=ring_01_r}}
    ]
    constraintWeights [1,1,1,1,1,1,1,1]
   }
   n30 = Get bIsPlayingEmote()
   n31 = "Transform (Modify) Bone - Bone: dyn_claws_base_l"!(ComponentPose, bAlphaBoolEnabled:bool=@n30, Translation="0,0,0", Rotation="0,10,0", Scale="1,1,1") -> (Pose) {
    alphaInputType=Bool
    rotationMode=BMM_Additive
    rotationSpace=BCS_BoneSpace
    translationMode=BMM_Additive
    translationSpace=BCS_BoneSpace
    alphaBoolBlend {blendInTime=0.1, blendOption=Linear, blendOutTime=0.1}
    boneToModify {boneName=dyn_claws_base_l}
    rotation {yaw=10}
   }
   "Transform (Modify) Bone - Bone: dyn_claws_base_r"!(ComponentPose=@n31, bAlphaBoolEnabled:bool=@n30, Translation="0,0,0", Rotation="-2,5,0", Scale="1,1,1") -> (Pose) {
    alphaInputType=Bool
    rotationMode=BMM_Additive
    rotationSpace=BCS_BoneSpace
    translationMode=BMM_Additive
    translationSpace=BCS_BoneSpace
    alphaBoolBlend {blendInTime=0.1, blendOption=Linear, blendOutTime=0.1}
    boneToModify {boneName=dyn_claws_base_r}
    rotation {pitch=-2, yaw=5}
   }
   n32 = "Constraint - dyn_belt_f_01 (P)"!(ComponentPose, bAlphaBoolEnabled:bool=@n5, ConstraintWeights_0=1) -> (Pose) {
    alphaInputType=Bool
    alphaBoolBlend {blendOption=Linear}
    boneToModify {boneName=dyn_belt_f_01}
    constraintSetup [{offsetOption=None, targetBone {boneName=dyn_belt_f_01_AIM}}]
    constraintWeights [1]
   }
   n33 = "Constraint - dyn_belt_l_01 (P)"!(ComponentPose=@n32, bAlphaBoolEnabled:bool=@n5, ConstraintWeights_0=1) -> (Pose) {
    alphaInputType=Bool
    alphaBoolBlend {blendOption=Linear}
    boneToModify {boneName=dyn_belt_l_01}
    constraintSetup [{offsetOption=None, targetBone {boneName=dyn_belt_l_01_AIM}}]
    constraintWeights [1]
   }
   n34 = "Constraint - dyn_belt_r_01 (P)"!(ComponentPose=@n33, bAlphaBoolEnabled:bool=@n5, ConstraintWeights_0=1) -> (Pose) {
    alphaInputType=Bool
    alphaBoolBlend {blendOption=Linear}
    boneToModify {boneName=dyn_belt_r_01}
    constraintSetup [{offsetOption=None, targetBone {boneName=dyn_belt_r_01_AIM}}]
    constraintWeights [1]
   }
   n35 = "Constraint - dyn_belt_b_01 (P)"!(ComponentPose=@n34, bAlphaBoolEnabled:bool=@n5, ConstraintWeights_0=1) -> (Pose) {
    alphaInputType=Bool
    alphaBoolBlend {blendOption=Linear}
    boneToModify {boneName=dyn_belt_b_01}
    constraintSetup [{offsetOption=None, targetBone {boneName=dyn_belt_b_01_AIM}}]
    constraintWeights [1]
   }
   n36 = "Constraint - dyn_belt_f_01_AIM (PPPPPPPP)"!(
    ComponentPose=@n35
    bAlphaBoolEnabled:bool=@n5
    ConstraintWeights_0=0.2999
    ConstraintWeights_1:float=0.1647
    ConstraintWeights_2:float=0.2651
    ConstraintWeights_3:float=0.0976
    ConstraintWeights_4:float=0.0325
    ConstraintWeights_5:float=0.0058
    ConstraintWeights_6:float=0.0701
    ConstraintWeights_7:float=0.0644
   ) -> (Pose) {
    alphaInputType=Bool
    alphaBoolBlend {blendOption=Linear}
    boneToModify {boneName=dyn_belt_f_01_AIM}
    constraintSetup [
     {targetBone {boneName=pelvis}}
     {targetBone {boneName=spine_01}}
     {targetBone {boneName=spine_02}}
     {targetBone {boneName=spine_03}}
     {targetBone {boneName=spine_04}}
     {targetBone {boneName=thigh_l}}
     {targetBone {boneName=thigh_twist_01_l}}
     {targetBone {boneName=thigh_twist_01_r}}
    ]
    constraintWeights [1,1,1,1,1,1,1,1]
   }
   n37 = "Constraint - dyn_belt_l_01_AIM (PPPPPPPP)"(
    ComponentPose=@n36
    bAlphaBoolEnabled=@n5
    ConstraintWeights_0=0.3382
    ConstraintWeights_1=0.2586
    ConstraintWeights_2=0.1798
    ConstraintWeights_3=0.0423
    ConstraintWeights_4=0.006
    ConstraintWeights_5=0.0322
    ConstraintWeights_6=0.1322
    ConstraintWeights_7=0.0107
   ) {
    alphaInputType=Bool
    alphaBoolBlend {blendOption=Linear}
    boneToModify {boneName=dyn_belt_l_01_AIM}
    constraintSetup [
     {targetBone {boneName=pelvis}}
     {targetBone {boneName=spine_01}}
     {targetBone {boneName=spine_02}}
     {targetBone {boneName=spine_03}}
     {targetBone {boneName=spine_04}}
     {targetBone {boneName=thigh_l}}
     {targetBone {boneName=thigh_twist_01_l}}
     {targetBone {boneName=thigh_twist_01_r}}
    ]
    constraintWeights [1,1,1,1,1,1,1,1]
   }
   n38 = "Constraint - dyn_belt_r_01_AIM (PPPPPPPP)"!(
    ComponentPose=@n37
    bAlphaBoolEnabled:bool=@n5
    ConstraintWeights_0=0.2245
    ConstraintWeights_1:float=0.2896
    ConstraintWeights_2:float=0.276
    ConstraintWeights_3:float=0.0729
    ConstraintWeights_4:float=0.0088
    ConstraintWeights_5:float=0.0249
    ConstraintWeights_6:float=0.0976
    ConstraintWeights_7:float=0.0057
   ) -> (Pose) {
    alphaInputType=Bool
    alphaBoolBlend {blendOption=Linear}
    boneToModify {boneName=dyn_belt_r_01_AIM}
    constraintSetup [
     {targetBone {boneName=pelvis}}
     {targetBone {boneName=spine_01}}
     {targetBone {boneName=spine_02}}
     {targetBone {boneName=spine_03}}
     {targetBone {boneName=spine_04}}
     {targetBone {boneName=thigh_l}}
     {targetBone {boneName=thigh_twist_01_l}}
     {targetBone {boneName=thigh_twist_01_r}}
    ]
    constraintWeights [1,1,1,1,1,1,1,1]
   }
   n39 = "Constraint - dyn_belt_b_01_AIM (PPPPPPP)"!(
    ComponentPose=@n38
    bAlphaBoolEnabled:bool=@n5
    ConstraintWeights_0=0.2746
    ConstraintWeights_1:float=0.3673
    ConstraintWeights_2:float=0.1134
    ConstraintWeights_3:float=0.2258
    ConstraintWeights_4:float=0.0008
    ConstraintWeights_5:float=0.0091
    ConstraintWeights_6:float=0.0091
   ) -> (Pose) {
    alphaInputType=Bool
    alphaBoolBlend {blendOption=Linear}
    boneToModify {boneName=dyn_belt_b_01_AIM}
    constraintSetup [
     {targetBone {boneName=pelvis}}
     {targetBone {boneName=spine_01}}
     {targetBone {boneName=spine_02}}
     {targetBone {boneName=spine_03}}
     {targetBone {boneName=spine_04}}
     {targetBone {boneName=thigh_twist_01_l}}
     {targetBone {boneName=thigh_twist_01_r}}
    ]
    constraintWeights [1,1,1,1,1,1,1]
   }
   n40 = AND Boolean(A=Get bIsCrouching(), B=Get bIsMoving2D())
   n41 = "Transform (Modify) Bone - Bone: dyn_belt_l_01"!(ComponentPose=@n39, bAlphaBoolEnabled:bool=@n40, Translation="3,0,1", Rotation="0,0,0", Scale="1,1,1") -> (Pose) {
    alphaInputType=Bool
    rotationMode=BMM_Additive
    rotationSpace=BCS_BoneSpace
    scaleMode=BMM_Additive
    scaleSpace=BCS_BoneSpace
    translationMode=BMM_Additive
    translationSpace=BCS_BoneSpace
    alphaBoolBlend {blendOption=Linear}
    boneToModify {boneName=dyn_belt_l_01}
    translation {x=3, z=1}
   }
   n42 = "Transform (Modify) Bone - Bone: dyn_belt_f_01"!(ComponentPose=@n41, bAlphaBoolEnabled:bool=@n40, Translation="3,2,0", Rotation="0,-30,0", Scale="1,1,1") -> (Pose) {
    alphaInputType=Bool
    rotationMode=BMM_Additive
    rotationSpace=BCS_BoneSpace
    scaleMode=BMM_Additive
    scaleSpace=BCS_BoneSpace
    translationMode=BMM_Additive
    translationSpace=BCS_BoneSpace
    alphaBoolBlend {blendOption=Linear}
    boneToModify {boneName=dyn_belt_f_01}
    rotation {yaw=-30}
    translation=<3 2>
   }
   n43 = "Transform (Modify) Bone - Bone: dyn_belt_r_01"!(ComponentPose=@n42, bAlphaBoolEnabled:bool=@n40, Translation="2,2,-1", Rotation="0,0,0", Scale="1,1,1") -> (Pose) {
    alphaInputType=Bool
    rotationMode=BMM_Additive
    rotationSpace=BCS_BoneSpace
    scaleMode=BMM_Additive
    scaleSpace=BCS_BoneSpace
    translationMode=BMM_Additive
    translationSpace=BCS_BoneSpace
    alphaBoolBlend {blendOption=Linear}
    boneToModify {boneName=dyn_belt_r_01}
    translation=<2 2 -1>
   }
   "Transform (Modify) Bone - Bone: dyn_belt_b_01"!(ComponentPose=@n43, bAlphaBoolEnabled:bool=@n40, Translation="0,0,0", Rotation="0,0,0", Scale="1,1,1.1") -> (Pose) {
    alphaInputType=Bool
    rotationMode=BMM_Additive
    rotationSpace=BCS_BoneSpace
    scaleMode=BMM_Additive
    scaleSpace=BCS_BoneSpace
    translationMode=BMM_Additive
    translationSpace=BCS_BoneSpace
    alphaBoolBlend {blendOption=Linear}
    boneToModify {boneName=dyn_belt_b_01}
    scale {z=1.1}
   }
   n44 = Get Left_leg_belt_crouch() -> (Left_leg_belt_crouch:float)
   n45 = Get ClawsConstraint() -> (ClawsConstraint:bool)
   n46 = "Constraint - clawbase_01_r (PP)"!(
    ComponentPose=Local To Component(LocalPose=Copy Pose From Mesh() {bUseAttachedParent=true})
    bAlphaBoolEnabled:bool=@n45
    ConstraintWeights_0=0.5
    ConstraintWeights_1:float=0.5
   ) -> (Pose) {
    alphaInputType=Bool
    alphaBoolBlend {blendOption=Linear}
    boneToModify {boneName=clawbase_01_r}
    constraintSetup [{targetBone {boneName=index_metacarpal_r}}, {targetBone {boneName=middle_metacarpal_r}}]
    constraintWeights [1,1]
   }
   n47 = "Constraint - clawbase_02_r (PP)"!(ComponentPose=@n46, bAlphaBoolEnabled:bool=@n45, ConstraintWeights_0=0.5, ConstraintWeights_1:float=0.5) -> (Pose) {
    alphaInputType=Bool
    alphaBoolBlend {blendOption=Linear}
    boneToModify {boneName=clawbase_02_r}
    constraintSetup [{targetBone {boneName=middle_metacarpal_r}}, {targetBone {boneName=ring_metacarpal_r}}]
    constraintWeights [1,1]
   }
   n48 = "Constraint - clawbase_03_r (PP)"!(ComponentPose=@n47, bAlphaBoolEnabled:bool=@n45, ConstraintWeights_0=0.5, ConstraintWeights_1:float=0.5) -> (Pose) {
    alphaInputType=Bool
    alphaBoolBlend {blendOption=Linear}
    boneToModify {boneName=clawbase_03_r}
    constraintSetup [{targetBone {boneName=ring_metacarpal_r}}, {targetBone {boneName=pinky_metacarpal_r}}]
    constraintWeights [1,1]
   }
   n49 = "Constraint - clawbase_01_l (PP)"!(ComponentPose=@n48, bAlphaBoolEnabled:bool=@n45, ConstraintWeights_0=0.5, ConstraintWeights_1:float=0.5) -> (Pose) {
    alphaInputType=Bool
    alphaBoolBlend {blendOption=Linear}
    boneToModify {boneName=clawbase_01_l}
    constraintSetup [{targetBone {boneName=index_metacarpal_l}}, {targetBone {boneName=middle_metacarpal_l}}]
    constraintWeights [1,1]
   }
   n50 = "Constraint - clawbase_02_l (PP)"!(ComponentPose=@n49, bAlphaBoolEnabled:bool=@n45, ConstraintWeights_0=0.5, ConstraintWeights_1:float=0.5) -> (Pose) {
    alphaInputType=Bool
    alphaBoolBlend {blendOption=Linear}
    boneToModify {boneName=clawbase_02_l}
    constraintSetup [{targetBone {boneName=middle_metacarpal_l}}, {targetBone {boneName=ring_metacarpal_l}}]
    constraintWeights [1,1]
   }
   n51 = "Constraint - clawbase_03_l (PP)"!(ComponentPose=@n50, bAlphaBoolEnabled:bool=@n45, ConstraintWeights_0=0.5, ConstraintWeights_1:float=0.5) -> (Pose) {
    alphaInputType=Bool
    alphaBoolBlend {blendOption=Linear}
    boneToModify {boneName=clawbase_03_l}
    constraintSetup [{targetBone {boneName=ring_metacarpal_l}}, {targetBone {boneName=pinky_metacarpal_l}}]
    constraintWeights [1,1]
   }
   n52 = Get Claws open() -> (Claws open:bool)
   n53 = "Transform (Modify) Bone - Bone: clawbase_01_r"!(ComponentPose=@n51, bAlphaBoolEnabled:bool=@n52, Translation="0.5,0.1,-0.1", Rotation="0,-1,0", Scale="1,1,1") -> (Pose) {
    alphaInputType=Bool
    rotationMode=BMM_Additive
    rotationSpace=BCS_BoneSpace
    scaleMode=BMM_Additive
    scaleSpace=BCS_BoneSpace
    translationMode=BMM_Additive
    translationSpace=BCS_BoneSpace
    alphaBoolBlend {blendInTime=0.1, blendOption=Linear, blendOutTime=0.1}
    boneToModify {boneName=clawbase_01_r}
    rotation {yaw=-1}
    translation=<0.5 0.1 -0.1>
   }
   n54 = "Transform (Modify) Bone - Bone: clawbase_02_r"!(ComponentPose=@n53, bAlphaBoolEnabled:bool=@n52, Translation="0.5,0,0.1", Rotation="0,0,0", Scale="1,1,1") -> (Pose) {
    alphaInputType=Bool
    rotationMode=BMM_Additive
    rotationSpace=BCS_BoneSpace
    scaleMode=BMM_Additive
    scaleSpace=BCS_BoneSpace
    translationMode=BMM_Additive
    translationSpace=BCS_BoneSpace
    alphaBoolBlend {blendInTime=0.1, blendOption=Linear, blendOutTime=0.1}
    boneToModify {boneName=clawbase_02_r}
    translation {x=0.5, z=0.1}
   }
   n55 = "Transform (Modify) Bone - Bone: clawbase_03_r"!(ComponentPose=@n54, bAlphaBoolEnabled:bool=@n52, Translation="0.6,0,-0.1", Rotation="0,0,0", Scale="1,1,1") -> (Pose) {
    alphaInputType=Bool
    rotationMode=BMM_Additive
    rotationSpace=BCS_BoneSpace
    scaleMode=BMM_Additive
    scaleSpace=BCS_BoneSpace
    translationMode=BMM_Additive
    translationSpace=BCS_BoneSpace
    alphaBoolBlend {blendInTime=0.1, blendOption=Linear, blendOutTime=0.1}
    boneToModify {boneName=clawbase_03_r}
    translation {x=0.6, z=-0.1}
   }
   n56 = "Transform (Modify) Bone - Bone: clawbase_01_l"!(ComponentPose=@n55, bAlphaBoolEnabled:bool=@n52, Translation="-0.5,-0.1,0.2", Rotation="0,0,0", Scale="1,1,1") -> (Pose) {
    alphaInputType=Bool
    rotationMode=BMM_Additive
    rotationSpace=BCS_BoneSpace
    scaleMode=BMM_Additive
    scaleSpace=BCS_BoneSpace
    translationMode=BMM_Additive
    translationSpace=BCS_BoneSpace
    alphaBoolBlend {blendInTime=0.1, blendOption=Linear, blendOutTime=0.1}
    boneToModify {boneName=clawbase_01_l}
    translation=<-0.5 -0.1 0.2>
   }
   n57 = "Transform (Modify) Bone - Bone: clawbase_02_l"!(ComponentPose=@n56, bAlphaBoolEnabled:bool=@n52, Translation="-0.5,0,0", Rotation="0,0,0", Scale="1,1,1") -> (Pose) {
    alphaInputType=Bool
    rotationMode=BMM_Additive
    rotationSpace=BCS_BoneSpace
    scaleMode=BMM_Additive
    scaleSpace=BCS_BoneSpace
    translationMode=BMM_Additive
    translationSpace=BCS_BoneSpace
    alphaBoolBlend {blendInTime=0.1, blendOption=Linear, blendOutTime=0.1}
    boneToModify {boneName=clawbase_02_l}
    translation {x=-0.5}
   }
   n58 = "Transform (Modify) Bone - Bone: clawbase_03_l"!(ComponentPose=@n57, bAlphaBoolEnabled:bool=@n52, Translation="-0.5,0,0", Rotation="0,0,0", Scale="1,1,1") -> (Pose) {
    alphaInputType=Bool
    rotationMode=BMM_Additive
    rotationSpace=BCS_BoneSpace
    scaleMode=BMM_Additive
    scaleSpace=BCS_BoneSpace
    translationMode=BMM_Additive
    translationSpace=BCS_BoneSpace
    alphaBoolBlend {blendInTime=0.1, blendOption=Linear, blendOutTime=0.1}
    boneToModify {boneName=clawbase_03_l}
    translation {x=-0.5}
   }
   n59 = "Constraint - dyn_belt_f_01 (PPPPPPPP)"(
    ComponentPose=@n58
    bAlphaBoolEnabled=True
    ConstraintWeights_0=0.4639
    ConstraintWeights_1=0.1114
    ConstraintWeights_2=0.108
    ConstraintWeights_3=0.031
    ConstraintWeights_4=0.0123
    ConstraintWeights_5=0.1337
    ConstraintWeights_6=0.0116
    ConstraintWeights_7=0.128
   ) {
    alphaInputType=Bool
    alphaBoolBlend {blendOption=Linear}
    boneToModify {boneName=dyn_belt_f_01}
    constraintSetup [
     {targetBone {boneName=pelvis}}
     {targetBone {boneName=spine_01}}
     {targetBone {boneName=spine_02}}
     {targetBone {boneName=spine_03}}
     {targetBone {boneName=thigh_l}}
     {targetBone {boneName=thigh_twist_01_l}}
     {targetBone {boneName=thigh_r}}
     {targetBone {boneName=thigh_twist_01_r}}
    ]
    constraintWeights [1,1,1,1,1,1,1,1]
   }
   n60 = "Constraint - dyn_belt_r_01 (PPPPPPPP)"!(
    ComponentPose=@n59
    bAlphaBoolEnabled:bool=True
    ConstraintWeights_0=0.4526
    ConstraintWeights_1=0.1574
    ConstraintWeights_2=0.0914
    ConstraintWeights_3=0.0174
    ConstraintWeights_4=0.0033
    ConstraintWeights_5=0.0231
    ConstraintWeights_6=0.0508
    ConstraintWeights_7=0.2041
   ) -> (Pose) {
    alphaInputType=Bool
    alphaBoolBlend {blendOption=Linear}
    boneToModify {boneName=dyn_belt_r_01}
    constraintSetup [
     {targetBone {boneName=pelvis}}
     {targetBone {boneName=spine_01}}
     {targetBone {boneName=spine_02}}
     {targetBone {boneName=spine_03}}
     {targetBone {boneName=spine_04}}
     {targetBone {boneName=thigh_twist_01_l}}
     {targetBone {boneName=thigh_r}}
     {targetBone {boneName=thigh_twist_01_r}}
    ]
    constraintWeights [1,1,1,1,1,1,1,1]
   }
   n61 = "Constraint - dyn_belt_l_01 (PPPPPPPP)"!(
    ComponentPose=@n60
    bAlphaBoolEnabled:bool=True
    ConstraintWeights_0=0.4819
    ConstraintWeights_1=0.1385
    ConstraintWeights_2=0.0738
    ConstraintWeights_3=0.0152
    ConstraintWeights_4=0.0026
    ConstraintWeights_5=0.0429
    ConstraintWeights_6=0.2256
    ConstraintWeights_7=0.0195
   ) -> (Pose) {
    alphaInputType=Bool
    alphaBoolBlend {blendOption=Linear}
    boneToModify {boneName=dyn_belt_l_01}
    constraintSetup [
     {targetBone {boneName=pelvis}}
     {targetBone {boneName=spine_01}}
     {targetBone {boneName=spine_02}}
     {targetBone {boneName=spine_03}}
     {targetBone {boneName=spine_04}}
     {targetBone {boneName=thigh_l}}
     {targetBone {boneName=thigh_twist_01_l}}
     {targetBone {boneName=thigh_twist_01_r}}
    ]
    constraintWeights [1,1,1,1,1,1,1,1]
   }
   n62 = "Transform (Modify) Bone - Bone: dyn_belt_l_01"!(ComponentPose=@n61, Alpha:float=@n44, Translation="-15,15,5", Rotation="-15,0,0", Scale="1,1,1") -> (Pose) {
    rotationMode=BMM_Additive
    rotationSpace=BCS_BoneSpace
    translationMode=BMM_Additive
    translationSpace=BCS_BoneSpace
    alphaBoolBlend {blendOption=Linear}
    boneToModify {boneName=dyn_belt_l_01}
    rotation {pitch=-15}
    translation=<-15 15 5>
   }
   n63 = "Transform (Modify) Bone - Bone: dyn_belt_f_01"!(ComponentPose=@n62, Alpha:float=@n44, Translation="-2.5,10,0", Rotation="18,0,0", Scale="1,1,1") -> (Pose) {
    rotationMode=BMM_Additive
    rotationSpace=BCS_BoneSpace
    translationMode=BMM_Additive
    translationSpace=BCS_BoneSpace
    alphaBoolBlend {blendOption=Linear}
    boneToModify {boneName=dyn_belt_f_01}
    rotation {pitch=18}
    translation=<-2.5 10>
   }
   n64 = "Transform (Modify) Bone - Bone: dyn_belt_r_01"!(
    ComponentPose=@n63
    Alpha:float=Get Right_leg_belt_crouch() -> (Right_leg_belt_crouch:double)
    Translation="-15,5,-10"
    Rotation="10,0,0"
    Scale="1,1,1"
   ) -> (Pose) {
    rotationMode=BMM_Additive
    rotationSpace=BCS_BoneSpace
    translationMode=BMM_Additive
    translationSpace=BCS_BoneSpace
    alphaBoolBlend {blendOption=Linear}
    boneToModify {boneName=dyn_belt_r_01}
    rotation {pitch=10}
    translation=<-15 5 -10>
   }
   n65 = "Transform (Modify) Bone - Bone: dyn_belt_f_01"(ComponentPose=@n64, Alpha=Get Right_leg_belt_crouch(), Translation="-5,2,0", Rotation="-20,0,0", Scale="1,1,1") {
    rotationMode=BMM_Additive
    rotationSpace=BCS_BoneSpace
    translationMode=BMM_Additive
    translationSpace=BCS_BoneSpace
    alphaBoolBlend {blendOption=Linear}
    boneToModify {boneName=dyn_belt_f_01}
    rotation {pitch=-20}
    translation=<-5 2>
   }
   n66 = Control Rig!(Source=Component To Local(ComponentPose=@n65), Alpha, Is Sprinting:double=Get Is Sprinting() -> (Is Sprinting:double)) -> (Pose) {
    alphaBoolBlend {blendOption=Linear}
    controlRigAssetReference {blueprintRigClass=&/M_MED_TroutWrist_Spine_CtrlRig.M_MED_TroutWrist_Spine_CtrlRig_C}
   }
   Output Pose(Result=@n66) {layerGroup=None}
   NOT Boolean(A=Get bIsCrouching())
   OR Boolean(A=false, B=false)
   n67 = Get bIsCrouching()
   n68 = "Transform (Modify) Bone - Bone: dyn_belt_f_01"!(ComponentPose, bAlphaBoolEnabled:bool=@n67, Translation="-5,-3,0", Rotation="0,-20,0", Scale="1,1,1") -> (Pose) {
    alphaInputType=Bool
    rotationMode=BMM_Additive
    rotationSpace=BCS_BoneSpace
    scaleMode=BMM_Additive
    scaleSpace=BCS_BoneSpace
    translationMode=BMM_Additive
    translationSpace=BCS_BoneSpace
    alphaBoolBlend {blendOption=Linear}
    boneToModify {boneName=dyn_belt_f_01}
    rotation {yaw=-20}
    translation=<-5 -3>
   }
   n69 = "Transform (Modify) Bone - Bone: dyn_belt_l_01"!(ComponentPose=@n68, bAlphaBoolEnabled:bool=@n67, Translation="-5,-1,0", Rotation="0,0,0", Scale="1,1,1") -> (Pose) {
    alphaInputType=Bool
    rotationMode=BMM_Additive
    rotationSpace=BCS_BoneSpace
    scaleMode=BMM_Additive
    scaleSpace=BCS_BoneSpace
    translationMode=BMM_Additive
    translationSpace=BCS_BoneSpace
    alphaBoolBlend {blendOption=Linear}
    boneToModify {boneName=dyn_belt_l_01}
    translation=<-5 -1>
   }
   n70 = "Transform (Modify) Bone - Bone: dyn_belt_r_01"!(ComponentPose=@n69, bAlphaBoolEnabled:bool=@n67, Translation="-5,-1,0", Rotation="0,0,0", Scale="1,1,1") -> (Pose) {
    alphaInputType=Bool
    rotationMode=BMM_Additive
    rotationSpace=BCS_BoneSpace
    scaleMode=BMM_Additive
    scaleSpace=BCS_BoneSpace
    translationMode=BMM_Additive
    translationSpace=BCS_BoneSpace
    alphaBoolBlend {blendOption=Linear}
    boneToModify {boneName=dyn_belt_r_01}
    translation=<-5 -1>
   }
   n71 = "Transform (Modify) Bone - Bone: dyn_belt_b_01"(ComponentPose=@n70, bAlphaBoolEnabled=@n67, Translation="-5,-1,0", Rotation="0,0,0", Scale="1,1,1") {
    alphaInputType=Bool
    rotationMode=BMM_Additive
    rotationSpace=BCS_BoneSpace
    scaleMode=BMM_Additive
    scaleSpace=BCS_BoneSpace
    translationMode=BMM_Additive
    translationSpace=BCS_BoneSpace
    alphaBoolBlend {blendOption=Linear}
    boneToModify {boneName=dyn_belt_b_01}
    translation=<-5 -1>
   }
   n72 = "Transform (Modify) Bone - Bone: dyn_belt_f_01"(
    ComponentPose=@n71
    bAlphaBoolEnabled=AND Boolean(A=Get bIsCrouching(), B=Get bIsMoving2D())
    Translation="3,2,0"
    Rotation="0,-10,0"
    Scale="1,1,1"
   ) {
    alphaInputType=Bool
    rotationMode=BMM_Additive
    rotationSpace=BCS_BoneSpace
    scaleMode=BMM_Additive
    scaleSpace=BCS_BoneSpace
    translationMode=BMM_Additive
    translationSpace=BCS_BoneSpace
    alphaBoolBlend {blendOption=Linear}
    boneToModify {boneName=dyn_belt_f_01}
    rotation {yaw=-10}
    translation=<3 2>
   }
   Reroute Node(InputPin:Component Space Pose Link Structure=@n72) -> (OutputPin:Component Space Pose Link Structure)
   Get bIsSkyDiving()
  ]}
 EventGraph {nodes [
   n1 = Event BlueprintUpdateAnimation()
   > Set Delta Time X(Delta Time X=@n1.DeltaTimeX)
   > Set Right_leg_belt_crouch(Right_leg_belt_crouch:double=GetAlphaFromReferenceTransformDelta(SkeletalMeshComponent=Get Skel Mesh(), BoneName=thigh_r, BaseName=pelvis, InputRange=(X=-18,Y=-98), TwistAxis="1,0,0")) -> (Output_Get:double)
   > Set Left_leg_belt_crouch(Left_leg_belt_crouch:float=GetAlphaFromReferenceTransformDelta(SkeletalMeshComponent=Get Skel Mesh(), BoneName=thigh_l, BaseName=pelvis, InputRange=(X=-18,Y=-98), TwistAxis="1,0,0")) -> (Output_Get:float)
   n2 = Get bIsTacticalSprinting()
   > Set Is Sprinting(Is Sprinting:double=FInterpTo(Current=Get Is Sprinting(), Target="To Float (Boolean)"(InBool=@n2), DeltaTime=Get Delta Time X(), InterpSpeed=10)) -> (Output_Get:double)
   > n3 = Cast To FortPlayerAnimInstance(Object=GetAnimInstance(self=GetOwnerSkeletalMeshComponent()))
   > n4 = Switch on EFortWeaponCoreAnimation!(execute, Selection:EFortWeaponCoreAnimation Enum=Get WeaponCoreAnim(self=@n3)) -> (
    Melee:Exec
    Pistol:Exec
    Shotgun:Exec
    PaperBlueprint:Exec
    Rifle:Exec
    MeleeOneHand:Exec
    MachinePistol:Exec
    RocketLauncher:Exec
    GrenadeLauncher:Exec
    GoingCommando:Exec
    AssaultRifle:Exec
    TacticalShotgun:Exec
    SniperRifle:Exec
    TrapPlacement:Exec
    ShoulderLauncher:Exec
    AbilityDecoTool:Exec
    Crossbow:Exec
    C4:Exec
    RemoteControl:Exec
    DualWield:Exec
    AR_BullPup:Exec
    AR_ForwardGrip:Exec
    MedPackPaddles:Exec
    SMG_P90:Exec
    AR_DrumGun:Exec
    Consumable_Small:Exec
    Consumable_Large:Exec
    Balloon:Exec
    MountedTurret:Exec
    CreativeTool:Exec
    ExplosiveBow:Exec
    AshtonIndigo:Exec
    AshtonChicago:Exec
    MeleeDualWield:Exec
    Unarmed:Exec
   )
   n5 = Set isNOTMelee(
    execute=[
     @n4.AR_BullPup
     @n4.AR_DrumGun
     @n4.AR_ForwardGrip
     @n4.AbilityDecoTool
     @n4.AshtonChicago
     @n4.AshtonIndigo
     @n4.AssaultRifle
     @n4.Balloon
     @n4.C4
     @n4.Consumable_Large
     @n4.Consumable_Small
     @n4.CreativeTool
     @n4.Crossbow
     @n4.DualWield
     @n4.ExplosiveBow
     @n4.GoingCommando
     @n4.GrenadeLauncher
     @n4.MachinePistol
     @n4.MedPackPaddles
     @n4.MountedTurret
     @n4.PaperBlueprint
     @n4.Pistol
     @n4.RemoteControl
     @n4.Rifle
     @n4.RocketLauncher
     @n4.SMG_P90
     @n4.Shotgun
     @n4.ShoulderLauncher
     @n4.SniperRifle
     @n4.TacticalShotgun
     @n4.TrapPlacement
    ]
    isNOTMelee:bool=true
   ) -> (Output_Get:bool)
   > Set ClawsConstraint(ClawsConstraint:bool=OR Boolean!(A=@n5, B=Get bIsPlayingEmote(), C:bool=Get bIsSkyDiving()) -> (ReturnValue)) -> (Output_Get:bool)
   > Set Claws open(Claws open:bool=false) -> (Output_Get:bool)
   n6 = Set isNOTMelee(execute=[@n4.Melee, @n4.MeleeDualWield, @n4.MeleeOneHand], isNOTMelee=false)
   > Set ClawsConstraint(ClawsConstraint=OR Boolean(A=@n6, B=Get bIsPlayingEmote(), C=Get bIsSkyDiving()))
   > Set Claws open(Claws open=true)
   n7 = Event OnCharacterCustomizationLoaded()
   > Set Skel Mesh(Skel Mesh=Get Array Item with EFort Custom Part Type(Array=@n7.CustomMeshParts))
  ]}
]
variables [
 Claws open {type=boolean}
 ClawsConstraint {type=boolean}
 Delta Time X {type=number}
 Is Sprinting {type=number}
 isNOTMelee {type=boolean}
 Left_leg_belt_crouch {type=number}
 Right_leg_belt_crouch {type=number}
 Skel Mesh {type=SkeletalMeshComponent}
]
