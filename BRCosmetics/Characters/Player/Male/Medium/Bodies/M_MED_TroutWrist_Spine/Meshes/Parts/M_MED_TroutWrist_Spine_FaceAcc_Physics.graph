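# Physics asset for the TroutWrist Spine male character's face accessory, setting up physics for a mask's left and right ears.
# Each ear bone has a capsule constrained to a small capsule on its MSTR bone, with narrow, asymmetric swing and limited twist.
# A large unconstrained capsule on neck_01 supplies a body volume, likely for the ears to collide against.
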PhysicsAsset /BRCosmetics/Characters/Player/Male/Medium/Bodies/M_MED_TroutWrist_Spine/Meshes/Parts/M_MED_TroutWrist_Spine_FaceAcc_Physics
bodies {
 dyn_mask_ear_l [{shapeType=Capsule, length=3, radius=5, shapeName=dyn_mask_ear_l_capsule, center=<0 -6.8826 -1.8442>, rotation=<rot 0.0001 0.0002 -105>}]
 dyn_mask_ear_l_MSTR [{shapeType=Capsule, length=0.5, radius=0.5, shapeName=dyn_mask_ear_l_MSTR_capsule}]
 dyn_mask_ear_r [{shapeType=Capsule, length=3, radius=5, shapeName=dyn_mask_ear_r_capsule, center=<0.0001 -6.1055 2.4476>, rotation=<rot -0.0003 0.0007 105>}]
 dyn_mask_ear_r_MSTR [{shapeType=Capsule, length=0.5, radius=0.5, shapeName=dyn_mask_ear_r_MSTR_capsule}]
 neck_01 [{shapeType=Capsule, length=40.75, radius=14.5, rotation=<rot 90 10.6197 23.0097>}]
}
constraints [
 {bone1Name=dyn_mask_ear_l, bone2Name=dyn_mask_ear_l_MSTR, swing1LimitDegrees=10, swing2LimitDegrees=1, twistLimitDegrees=20}
 {bone1Name=dyn_mask_ear_r, bone2Name=dyn_mask_ear_r_MSTR, swing1LimitDegrees=10, swing2LimitDegrees=1, twistLimitDegrees=20}
]
