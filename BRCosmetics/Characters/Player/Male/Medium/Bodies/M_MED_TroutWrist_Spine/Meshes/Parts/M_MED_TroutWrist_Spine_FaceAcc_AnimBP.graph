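# Animation Blueprint for the TroutWrist_Spine face accessory, which copies pose and curves from the attached parent mesh.
# Copy Pose From Mesh feeds Output Pose directly, leaving the simulation branch disconnected and its 'Presim' cache without input.
# That unused branch selects by bRigidBodyEnabled between a root-space RigidBody simulation and chained AnimDynamics on the left and right mask ear bones.
# The update event is empty.
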AnimBlueprint /BRCosmetics/Characters/Player/Male/Medium/Bodies/M_MED_TroutWrist_Spine/Meshes/Parts/M_MED_TroutWrist_Spine_FaceAcc_AnimBP
target_skeleton=&/M_MED_TroutWrist_Spine_FaceAcc_Skeleton
graphs [
 AnimGraph {nodes [
   n1 = RigidBody(ComponentPose=Local To Component(LocalPose=Use cached pose 'Presim'())) {
    bClampLinearTranslationLimitToRefPose=true
    bOverrideWorldGravity=true
    overridePhysicsAsset=&/M_MED_TroutWrist_Spine_FaceAcc_Physics
    simulationSpace=BaseBoneSpace
    alphaBoolBlend {blendOption=Linear}
    baseBoneRef {boneName=root}
    componentAppliedLinearAccClamp=<800 800 800>
    componentLinearAccScale=<0.65 0.65 0.65>
    componentLinearVelScale=<0.3 0.3 0.3>
    overrideWorldGravity {z=-980}
   }
   n2 = Component To Local(ComponentPose=@n1)
   n3 = Anim Dynamics - Bone: dyn_mask_ear_r(ComponentPose=Local To Component(LocalPose=Use cached pose 'Presim'()), Alpha=1, ExternalForce="0,0,0") {
    angularDampingOverride=0.999
    bOverrideAngularDamping=true
    bUseGravityOverride=true
    simulationSpace=RootRelative
    alphaBoolBlend {blendOption=Linear}
    boundBone {boneName=dyn_mask_ear_r}
    componentAppliedLinearAccClamp=<3000 3000 3000>
    componentLinearAccScale=<0.9 0.9 0.9>
    componentLinearVelScale=<1 1 1>
    gravityOverride {z=-980}
    physicsBodyDefinitions [{
      boundBone {boneName=dyn_mask_ear_r}
      boxExtents=<20 10 1>
      constraintSetup {angularLimitsMax=<30 3 5>, angularLimitsMin=<-30 -3 -5>, angularTarget=<0 0 0>}
      localJointOffset=<2 0 0>
     }]
   }
   n4 = Anim Dynamics - Bone: dyn_mask_ear_l(ComponentPose=@n3, Alpha=1, ExternalForce="0,0,0") {
    angularDampingOverride=0.999
    bOverrideAngularDamping=true
    bUseGravityOverride=true
    simulationSpace=RootRelative
    alphaBoolBlend {blendOption=Linear}
    boundBone {boneName=dyn_mask_ear_l}
    componentAppliedLinearAccClamp=<3000 3000 3000>
    componentLinearAccScale=<0.9 0.9 0.9>
    componentLinearVelScale=<1 1 1>
    gravityOverride {z=-980}
    physicsBodyDefinitions [{
      boundBone {boneName=dyn_mask_ear_l}
      boxExtents=<20 10 1>
      constraintSetup {angularLimitsMax=<30 3 5>, angularLimitsMin=<-30 -3 -5>, angularTarget=<0 0 0>}
      localJointOffset=<2 0 0>
     }]
   }
   n5 = Component To Local(ComponentPose=@n4)
   n6 = Get bRigidBodyEnabled()
   Component To Local(ComponentPose=Local To Component(LocalPose=Blend Poses by bool(BlendPose_0=@n2, BlendPose_1=@n5, bActiveValue=@n6) {blendType=Linear}))
   Output Pose(Result=Copy Pose From Mesh() {bCopyCurves=true, bUseAttachedParent=true}) {layerGroup=None}
   Save cached pose 'Presim'() {cacheName=Presim}
  ]}
 EventGraph {nodes [Event BlueprintUpdateAnimation()]}
]
variables []
